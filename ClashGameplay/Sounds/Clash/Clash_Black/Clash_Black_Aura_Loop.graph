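# Looping aura sound cue for Clash Black, layering a fading-in looped aura with one-shot start and equip sounds.
# An envelope then lowers the whole mix to a quieter level over about seven seconds.
# IsLocal and IsTeammate switches choose an unattenuated local mix, a quieter attenuated teammate mix, or a louder attenuated default mix for others.
# A concurrency override caps simultaneous instances at three, stopping the quietest.
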SoundCue /ClashGameplay/Sounds/Clash/Clash_Black/Clash_Black_Aura_Loop
root=Switch_0
nodes [
 {
  id=Attenuation_1
  children [Mixer_2]
  properties {attenuationSettings=/ClashGameplay/Gameplay/Clash/Shared/Audio/Aura/Default/Hightower_Aura_Friendly_Att}
 }
 {
  id=Attenuation_2
  children [Mixer_8]
  properties {attenuationSettings=/ClashGameplay/Gameplay/Clash/Shared/Audio/Aura/Default/Hightower_Aura_Att}
 }
 {
  id=Enveloper_0
  children [WavePlayer_0]
  properties {pitchCurve {editorCurveData {keys [{time=0}]}}, volumeCurve {editorCurveData {keys [<at 0.0117 0.0072>,<at 0.5073 1>]}}}
 }
 {
  id=Enveloper_3
  children [Mixer_0]
  properties {pitchCurve {editorCurveData {keys [{time=0}]}}, volumeCurve {editorCurveData {keys [<at 0 1>,<at 7.0425 0.2204>]}}}
 }
 {id=Mixer_0, children [Enveloper_0,WavePlayer_2,WavePlayer_3], properties {inputVolume [1.75,0.3,1.5]}}
 {id=Mixer_2, children [Enveloper_3], properties {inputVolume [0.5]}}
 {id=Mixer_6, children [Enveloper_3], properties {inputVolume [1]}}
 {id=Mixer_8, children [Enveloper_3], properties {inputVolume [1.5]}}
 {id=Switch_0, children [Switch_1,Switch_1,Mixer_6], properties {intParameterName=IsLocal}}
 {id=Switch_1, children [Attenuation_2, Attenuation_2, Attenuation_1], properties {intParameterName=IsTeammate}}
 {id=WavePlayer_0, properties {bLooping=true, soundWaveAssetPtr=/ClashGameplay/Gameplay/Clash/Shared/Audio/Vertigo_Aura_LP_01}}
 {id=WavePlayer_2, properties {soundWaveAssetPtr=/ClashGameplay/Gameplay/Clash/Shared/Audio/Aura/Default/Hightower_Aura_Start_01}}
 {id=WavePlayer_3, properties {soundWaveAssetPtr=/ClashGameplay/Gameplay/Clash/Shared/Audio/Vertigo_Equip_v02}}
]
properties {
 bOverrideConcurrency=true
 maxDistance=4000
 soundClassObject=&Fort_Player_Foley_Athena
 volumeMultiplier=1.2
 concurrencyOverrides {maxCount=3, resolutionRule=StopQuietest}
}
