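# Looping ambient Sound Cue for the Clash Black drop pod, played as a quiet in-game mission sound.
# A looping storm shield pulse hum wave feeds a Modulator that fixes pitch at half, lowering the borrowed hum.
# Close-focus prop attenuation is overridden with a short falloff, distance low-pass filtering and listener focus that ducks the sound when out of focus.
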SoundCue /ClashGameplay/Sounds/Clash/Clash_Black/Clash_Black_DropPod_Ambient_Loop_Cue
root=Modulator_0
nodes [
 {id=Modulator_0, children [WavePlayer_0], properties {pitchMax=0.5, pitchMin=0.5}}
 {id=WavePlayer_0, properties {bLooping=true, soundWaveAssetPtr=/Game/Sounds/Fort_Gadgets/Shield/ENRG_Storm_Shield_Pulse_Hum_02Loop}}
]
properties {
 attenuationSettings=/Game/Sounds/Attenuation/Fort_Prop_Close_Focus_Att
 bOverrideAttenuation=true
 maxDistance=1050
 soundClassObject=/Game/Sounds/FortSoundClassesAndModes/Fort_Mission_IngameSounds
 volumeMultiplier=0.125
 attenuationOverrides {
  bAttenuateWithLPF=true
  bEnableFocusInterpolation=true
  bEnableListenerFocus=true
  falloffDistance=800
  focusAttackInterpSpeed=2
  focusReleaseInterpSpeed=2
  nonFocusVolumeAttenuation=0.7
  attenuationShapeExtents {x=250}
 }
}
