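# Sound cue for the falling loop of the Clash Black drop pod, built from reused Meteor audio.
# It mixes a close layer (Shepard and close falling loops), a quieter very-distant loop, and a non-looping incoming sound.
# Both loop layers fade in through volume envelopers, and each of the three layers has its own attenuation node for distance-based blending.
# It uses the LTM gadgets sound class and meteor fall-impact concurrency, with debug enabled.
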SoundCue /ClashGameplay/Sounds/Clash/Clash_Black/Clash_Black_DropPod_Falling_Loop_Cue
root=Mixer_1
nodes [
 {
  id=Attenuation_3
  children [Mixer_3]
  properties {attenuationSettings=/Game/Sounds/Fort_GamePlay_Sounds/Meteor/Fort_Meteor_Carmine_Athena_Att}
 }
 {
  id=Attenuation_4
  children [WavePlayer_3]
  properties {bOverrideAttenuation=true, attenuationOverrides {falloffDistance=1500, stereoSpread=1000, attenuationShapeExtents {x=3000}}}
 }
 {
  id=Attenuation_5
  children [WavePlayer_7]
  properties {attenuationSettings=/Game/Sounds/Fort_GamePlay_Sounds/Meteor/Fort_Meteor_Carmine_Distant_Athena_Att}
 }
 {
  id=Enveloper_0
  children [Attenuation_3]
  properties {pitchCurve {editorCurveData {keys [{time=0}]}}, volumeCurve {editorCurveData {keys [<at 0 0>,<at 1 1>]}}}
 }
 {
  id=Enveloper_1
  children [Attenuation_5]
  properties {pitchCurve {editorCurveData {keys [{time=0}]}}, volumeCurve {editorCurveData {keys [<at 0 0>,<at 1 1>]}}}
 }
 {id=Mixer_1, children [Enveloper_0, Enveloper_1, Attenuation_4], properties {inputVolume [1.5,0.4,0.75]}}
 {id=Mixer_3, children [WavePlayer_6,WavePlayer_5], properties {inputVolume [3,0.5]}}
 {id=WavePlayer_3, properties {soundWaveAssetPtr=/Game/Sounds/Fort_GamePlay_Sounds/Meteor/Meteor_Carmine_Incoming_04}}
 {id=WavePlayer_5, properties {bLooping=true, soundWaveAssetPtr=/Game/Sounds/Fort_GamePlay_Sounds/Meteor/Meteor_Carmine_FallingLoop_Close_06}}
 {id=WavePlayer_6, properties {bLooping=true, soundWaveAssetPtr=/Game/Sounds/Fort_GamePlay_Sounds/Meteor/Meteor_Carmine_FallingLoop_Shepard}}
 {id=WavePlayer_7, properties {bLooping=true, soundWaveAssetPtr=/Game/Sounds/Fort_GamePlay_Sounds/Meteor/Meteor_Falling_VeryDistant_lp_01a}}
]
properties {
 bDebug=true
 maxDistance=2097152
 soundClassObject=/Game/Sounds/FortSoundClassesAndModes/Fort_Gadgets_LTM
 volumeMultiplier=0.4
 attenuationOverrides {falloffDistance=1000000, attenuationShapeExtents {x=0}}
 concurrencyOverrides {maxCount=2, resolutionRule=StopQuietest}
 concurrencySet [/Game/Sounds/Fort_GamePlay_Sounds/Meteor/Meteor_Rock_FallImpact_Conc]
}
