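# Sound cue for the Clash Black drop pod landing, mixing three layers routed to the Fort_Gadgets sound class.
# A distant supply-drop landing uses shared gadget attenuation, while a close landing hit is cut short by a volume envelope and low-pass filtered over a short falloff.
# A modulated Black/Red pod spawn sound carries across a very large custom falloff.
# The cue boosts overall volume and overrides concurrency to prevent new instances at the limit.
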SoundCue /ClashGameplay/Sounds/Clash/Clash_Black/Clash_Black_DropPod_Land_Cue
root=Mixer_0
nodes [
 {
  id=Attenuation_0
  children [Modulator_0]
  properties {bOverrideAttenuation=true, attenuationOverrides {falloffDistance=500000, stereoSpread=5000, attenuationShapeExtents {x=250000}}}
 }
 {id=Attenuation_1, children [SoundClass_1], properties {attenuationSettings=&Fort_GadgetActivate_Att}}
 {
  id=Attenuation_2
  children [SoundClass_0]
  properties {
   attenuationSettings=&Fort_GadgetActivate_Att
   bOverrideAttenuation=true
   attenuationOverrides {bAttenuateWithLPF=true, falloffDistance=2500, lPFFrequencyAtMax=1000, lPFRadiusMax=2500, lPFRadiusMin=400, stereoSpread=1000}
  }
 }
 {
  id=Enveloper_0
  children [WavePlayer_4]
  properties {
   pitchCurve {editorCurveData {keys [{time=0}]}}
   volumeCurve {editorCurveData {keys [<at 0 1>,<at 0.2649 1.0002>,<at 0.2968 0.9159>,<at 0.3107 0.7872>,<at 0.3311 0.4569>,<at 0.5427 -0.007>]}}
  }
 }
 {id=Mixer_0, children [Attenuation_1, Attenuation_2, SoundClass_2], properties {inputVolume [1,1,1]}}
 {id=Modulator_0, children [WavePlayer_1]}
 {id=SoundClass_0, children [Enveloper_0], properties {soundClassOverride=&Fort_Gadgets}}
 {id=SoundClass_1, children [WavePlayer_0], properties {soundClassOverride=&Fort_Gadgets}}
 {id=SoundClass_2, children [Attenuation_0], properties {soundClassOverride=&Fort_Gadgets}}
 {id=WavePlayer_0, properties {soundWaveAssetPtr=/Game/Sounds/Foley_Loot/Containers/SupplyDrops/sfx_supplydrop_land_distant}}
 {id=WavePlayer_1, properties {soundWaveAssetPtr=/ClashGameplay/Sounds/Clash/Clash_Red/Clash_Black_Red_Pod_Spawn_01}}
 {id=WavePlayer_4, properties {soundWaveAssetPtr=/Game/Sounds/Fort_Gadgets/Supply_Drop/supply_drop_land3}}
]
properties {bOverrideConcurrency=true, maxDistance=750000, volumeMultiplier=2, concurrencyOverrides {resolutionRule=PreventNew}}
