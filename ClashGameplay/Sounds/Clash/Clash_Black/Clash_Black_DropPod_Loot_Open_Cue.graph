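# Sound cue for opening loot from a Clash Black drop pod, as its name indicates.
# A mixer layers the Clash Black acquire sound with faction chest latch-open and eye-scan success foley, with the latch slightly quieter.
# Only the acquire layer passes through a modulator for slight random upward pitch variation.
# It uses item-interact attenuation, a loot reward sound class and a concurrency override that prevents new instances at the limit.
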SoundCue /ClashGameplay/Sounds/Clash/Clash_Black/Clash_Black_DropPod_Loot_Open_Cue
root=Mixer_0
nodes [
 {id=Mixer_0, children [Modulator_3,WavePlayer_0,WavePlayer_2], properties {inputVolume [1,0.75,1]}}
 {id=Modulator_3, children [WavePlayer_10], properties {pitchMax=1.1, pitchMin=1, volumeMax=1, volumeMin=1}}
 {id=WavePlayer_0, properties {soundWaveAssetPtr=/Game/Sounds/Foley_Loot/Containers/Faction_Chest/Faction_Chest_Latch_Open}}
 {id=WavePlayer_10, properties {soundWaveAssetPtr=&/Clash_Black_Acquire_01}}
 {id=WavePlayer_2, properties {soundWaveAssetPtr=/Game/Sounds/Foley_Loot/Containers/Faction_Chest/Faction_Chest_Eye_Scan_Success_01}}
]
properties {
 attenuationSettings=/Game/Sounds/Attenuation/Fort_Item_Interact_Att
 bOverrideConcurrency=true
 maxDistance=1850
 soundClassObject=&Fort_Loot_Reward_PSM
 attenuationOverrides {bEnableOcclusion=true, falloffDistance=1000, occlusionTraceChannel=ECC_GameTraceChannel11, attenuationShapeExtents {x=500}}
 concurrencyOverrides {resolutionRule=PreventNew}
}
