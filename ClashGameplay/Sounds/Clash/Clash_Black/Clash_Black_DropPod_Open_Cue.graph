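# Sound cue for the Clash Black drop pod opening, built from reused ceiling gas trap firing sounds.
# A Random node picks between two gas fire wave variants with equal weight.
# It plays at reduced volume through the trap placement sound class and trap attenuation, with overridden concurrency that prevents new instances once the limit is reached.
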SoundCue /ClashGameplay/Sounds/Clash/Clash_Black/Clash_Black_DropPod_Open_Cue
root=Random_27
nodes [
 {id=Random_27, children [WavePlayer_97, WavePlayer_98], properties {weights [1,1]}}
 {id=WavePlayer_97, properties {soundWaveAssetPtr=/Game/Sounds/Fort_Traps/Ceiling_Gas_Trap/CeilingTrap_Gas_Fire_002}}
 {id=WavePlayer_98, properties {soundWaveAssetPtr=/Game/Sounds/Fort_Traps/Ceiling_Gas_Trap/CeilingTrap_Gas_Fire_001}}
]
properties {
 attenuationSettings=&Fort_Traps_Att
 bOverrideConcurrency=true
 maxDistance=6000
 soundClassObject=/Game/Sounds/FortSoundClassesAndModes/Fort_Traps_Placement
 volumeMultiplier=0.5
 concurrencyOverrides {resolutionRule=PreventNew}
}
