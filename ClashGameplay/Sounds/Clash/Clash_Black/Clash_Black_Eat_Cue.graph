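# Sound cue for the Clash Black eat sound, layering an eat variation with a later roar tail.
# After a short fixed delay, a Mixer plays a random pick of two eat waves alongside an additionally delayed roar tail reused from the Hightower Vertigo emote.
# It uses the impacts sound class, borrows the player pickaxe impact attenuation and slightly boosts volume.
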SoundCue /ClashGameplay/Sounds/Clash/Clash_Black/Clash_Black_Eat_Cue
root=Delay_0
nodes [
 {id=Delay_0, children [Mixer_0], properties {delayMax=0.3, delayMin=0.3}}
 {id=Delay_1, children [WavePlayer_7], properties {delayMax=0.7, delayMin=0.7}}
 {id=Mixer_0, children [Delay_1,Random_1], properties {inputVolume [1,1]}}
 {id=Random_1, children [WavePlayer_5,WavePlayer_6], properties {weights [1,1]}}
 {id=WavePlayer_5, properties {soundWaveAssetPtr=&/Clash_Black_Eat_01}}
 {id=WavePlayer_6, properties {soundWaveAssetPtr=&/Clash_Black_Eat_02}}
 {id=WavePlayer_7, properties {soundWaveAssetPtr=/Game/Athena/Sounds/Emotes/Hightower_Vertigo/Vertigo_Emote_Roar_TEnd}}
]
properties {
 attenuationSettings=/Game/Sounds/Attenuation/Fort_Weapon_Pickaxe_Player_Impact_Att
 maxDistance=3512
 soundClassObject=&Fort_Impacts
 volumeMultiplier=1.125
 concurrencySet [null]
}
