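# Sound cue for Clash Black's punch impact, classed as a melee weapon sound.
# A mixer layers one of four equally weighted random hit waves, borrowed from the Clash Red folder, with shared pullback and impact whoosh waves.
# The mix feeds a root modulator with pitch and volume fixed at unity, so variation comes only from the random hit.
# Close melee attenuation applies, and a concurrency override allows one instance, blocking new plays while it sounds.
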SoundCue /ClashGameplay/Sounds/Clash/Clash_Black/Clash_Black_Punch_Impact
root=Modulator_1
nodes [
 {id=Mixer_1, children [Random_0,WavePlayer_0,WavePlayer_1], properties {inputVolume [1,1,1]}}
 {id=Modulator_1, children [Mixer_1], properties {pitchMax=1, pitchMin=1, volumeMax=1, volumeMin=1}}
 {id=Random_0, children [WavePlayer_2,WavePlayer_3,WavePlayer_4,WavePlayer_9], properties {weights [1,1,1,1]}}
 {id=WavePlayer_0, properties {soundWaveAssetPtr=/ClashGameplay/Gameplay/Clash/Shared/Audio/Vertigo_SNG_Pullback_v04}}
 {id=WavePlayer_1, properties {soundWaveAssetPtr=/ClashGameplay/Gameplay/Clash/Shared/Audio/Vertigo_SNG_Impact_Whoosh_01}}
 {id=WavePlayer_2, properties {soundWaveAssetPtr=/ClashGameplay/Sounds/Clash/Clash_Red/Clash_Black_Red_Hit_04}}
 {id=WavePlayer_3, properties {soundWaveAssetPtr=/ClashGameplay/Sounds/Clash/Clash_Red/Clash_Black_Red_Hit_01}}
 {id=WavePlayer_4, properties {soundWaveAssetPtr=/ClashGameplay/Sounds/Clash/Clash_Red/Clash_Black_Red_Hit_02}}
 {id=WavePlayer_9, properties {soundWaveAssetPtr=/ClashGameplay/Sounds/Clash/Clash_Red/Clash_Black_Red_Hit_03}}
]
properties {
 attenuationSettings=/Game/Sounds/Attenuation/Fort_Weapon_Umbrella_Close_Athena_Att
 bOverrideConcurrency=true
 maxDistance=4500
 soundClassObject=&Fort_WeaponMelee
 volumeMultiplier=0.5
 concurrencyOverrides {maxCount=1, resolutionRule=PreventNew}
}
