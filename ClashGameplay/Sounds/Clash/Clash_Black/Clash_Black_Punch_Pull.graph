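# Sound cue for the pull sound of the Clash Black punch, classed as a melee weapon sound.
# It picks one of three equally weighted pull waves at random, then a modulator applies a fixed volume cut with no pitch variation.
# A close-range umbrella weapon attenuation shapes falloff, and a concurrency override allows only one instance, blocking new plays while one is active.
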SoundCue /ClashGameplay/Sounds/Clash/Clash_Black/Clash_Black_Punch_Pull
root=Modulator_0
nodes [
 {id=Modulator_0, children [Random_0], properties {pitchMax=1, pitchMin=1, volumeMax=0.7, volumeMin=0.7}}
 {id=Random_0, children [WavePlayer_0,WavePlayer_3,WavePlayer_4], properties {weights [1,1,1]}}
 {id=WavePlayer_0, properties {soundWaveAssetPtr=&/Clash_Black_Pull_01}}
 {id=WavePlayer_3, properties {soundWaveAssetPtr=&/Clash_Black_Pull_02}}
 {id=WavePlayer_4, properties {soundWaveAssetPtr=&/Clash_Black_Pull_03}}
]
properties {
 attenuationSettings=/Game/Sounds/Attenuation/Fort_Weapon_Umbrella_Close_Athena_Att
 bOverrideConcurrency=true
 maxDistance=4500
 soundClassObject=&Fort_WeaponMelee
 volumeMultiplier=0.5
 concurrencyOverrides {maxCount=1, resolutionRule=PreventNew}
}
