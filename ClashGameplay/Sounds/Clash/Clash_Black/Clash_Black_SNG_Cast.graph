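# Sound Cue for the cast sound of the Clash Black SNG, played as a melee weapon sound.
# A Random node picks one of three cast wave variants with equal weight, feeding a Modulator whose pitch and volume ranges are fixed at unity, so it adds no variation.
# Overridden concurrency allows a single instance and blocks new plays, with close-range umbrella weapon attenuation and reduced volume.
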SoundCue /ClashGameplay/Sounds/Clash/Clash_Black/Clash_Black_SNG_Cast
root=Modulator_0
nodes [
 {id=Modulator_0, children [Random_0], properties {pitchMax=1, pitchMin=1, volumeMax=1, volumeMin=1}}
 {id=Random_0, children [WavePlayer_1,WavePlayer_2,WavePlayer_3], properties {weights [1,1,1]}}
 {id=WavePlayer_1, properties {soundWaveAssetPtr=&/Clash_Black_Cast_01}}
 {id=WavePlayer_2, properties {soundWaveAssetPtr=&/Clash_Black_Cast_02}}
 {id=WavePlayer_3, properties {soundWaveAssetPtr=&/Clash_Black_Cast_03}}
]
properties {
 attenuationSettings=/Game/Sounds/Attenuation/Fort_Weapon_Umbrella_Close_Athena_Att
 bOverrideConcurrency=true
 maxDistance=4500
 soundClassObject=&Fort_WeaponMelee
 volumeMultiplier=0.4
 concurrencyOverrides {maxCount=1, resolutionRule=PreventNew}
}
